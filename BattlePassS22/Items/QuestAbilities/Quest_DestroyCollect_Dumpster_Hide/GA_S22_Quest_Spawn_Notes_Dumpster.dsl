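; Season 22 quest gameplay ability that spawns collectible quest actors, apparently notes, from dumpster props for players on the quest.
; ShouldAbilityRespondToEvent accepts tag-matched building props, honoring the repeat-target setting, when the quest objective is in progress or a test-mode tag is present.
; SpawnQuestActorsFromDropEntry runs only while no valid spawned actor is stored, spawning the drop quantity offset from the prop, storing each, launching it and checking cosmetics.
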(event Ability|EventActivateAbility)

(event Ability|EventOnEndAbility
  (params (bWasCancelled :type Boolean)))

(fn ShouldAbilityRespondToEvent
  (params (ActorInfo :type Gameplay%20Ability%20Actor%20Info%20Structure) (Payload :type Gameplay%20Event%20Data%20Structure))
  (bind (_eventtag _instigator) (Utilities|Struct|BreakGameplayEventData Payload))
  (bind _asbuilding_prop (Utilities|Casting|CastToBuildingProp _instigator)
    (:then
      (if (and (or (Variables|Setup|GetCanTriggerMultipleTimesOnTarget) (!= (Variables|Data|GetStoredTarget) _instigator)) (GameplayTags|TagContainer|CompareTagAssetInterfacetoOtherTagContainers (Variables|Setup|GetTargetTags_Any) _asbuilding_prop))
        (bind _asfort_player_pawn (Utilities|Casting|CastToFortPlayerPawn (Ability|GetAvatarActorfromActorInfo))
          (:then
            (return (or (GameplayTags|TagContainer|CompareTagAssetInterfacetoOtherTagContainers (LocalVariables|GetTestModeTag) _asfort_player_pawn) (and (Quest|IsQuestinProgress (Quests|GetQuestManager (Fortnite|GetFortPlayerControllerfromActor _asfort_player_pawn) "Invalid") (Variables|Setup|GetQuestDefinition)) (Quest|IsObjectivewithNameinProgress (Quests|GetQuestManager (Fortnite|GetFortPlayerControllerfromActor _asfort_player_pawn) "Invalid") (Variables|Setup|GetQuestDefinition) (Variables|Setup|GetBackend%20Name) :!class /Script/FortniteGame.FortQuestManager)))))
          (:CastFailed
            (return false)))
        (else
          _)))
    (:CastFailed
      _))
  (return false))

(fn SpawnQuestActorsFromDropEntry
  (params (DropEntry :type Struct%20Spawned%20Quest%20Actor%20Drop%20Entry%20Structure) (EventData :type Gameplay%20Event%20Data%20Structure))
  (bind (_spawnedquestactorclass_6_f15cfde2424412e9badbe2975b317a07 _dropquantity_14_6469e6844f83772bb80669bcf263cb9d) (Utilities|Struct|BreakStructSpawnedQuestActorDropEntry DropEntry))
  (bind (_eventtag _instigator) (Utilities|Struct|BreakGameplayEventData EventData))
  (bind _owningplayerpawn (Variables|GAVars|Data|GetOwningPlayerPawn))
  (bind _returnvalue (Fortnite|GetFortPlayerControllerfromActor _owningplayerpawn))
  (Utilities|IsValid (Variables|Default|GetSpawnedActor)
    (:Is%20Not%20Valid
      (for _ (range 1 (+ _dropquantity_14_6469e6844f83772bb80669bcf263cb9d 1))
        (bind _returnvalue_2 (Building|SpawnBuildingGameplayActor _spawnedquestactorclass_6_f15cfde2424412e9badbe2975b317a07 (Math|Transform|MakeTransform (+ (.location (Spawning|GetSpawnTransform _instigator)) (* (+ (Transformation|GetActorRightVector _instigator) (Math|Vector|GetUpVector "0, 0, 0")) 1.5e+02)) (.rotation (Spawning|GetSpawnTransform _instigator))) _returnvalue))
        (bind _asbp_spawned_quest_actor_parent (Utilities|Casting|CastToBP_SpawnedQuestActor_Parent _returnvalue_2))
        (Variables|Default|SetSpawnedActor _asbp_spawned_quest_actor_parent)
        (Variables|/Game/Athena/Items/QuestParents/SpawnedQuestActor/BP_SpawnedQuestActor_Parent.BP_SpawnedQuestActor_Parent_C|SetbTestVisualsNoQuestRequired (GameplayTags|TagContainer|CompareTagAssetInterfacetoOtherTagContainers (LocalVariables|GetTestModeTag) _owningplayerpawn) _asbp_spawned_quest_actor_parent)
        (Projectile|SetVelocity (Variables|/Game/Athena/Items/QuestParents/SpawnedQuestActor/BP_SpawnedQuestActor_Parent.BP_SpawnedQuestActor_Parent_C|GetFortProjectileMovement _asbp_spawned_quest_actor_parent) (Spawning|GenerateSpawnVelocity))
        (Class|BPSpawnedQuestActorParent|CheckCosmetics _asbp_spawned_quest_actor_parent _returnvalue)))))
